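# Physics asset for the CountyFrog drum kit instrument cosmetic, giving physics shapes to the sea urchin hi-hat piece.
# It has three bodies: a sphere on the dyn_sea_urchin_hihat bone, a long capsule on hihat_top and a tiny capsule on handle.
# A single constraint ties the urchin to hihat_top with narrow swing limits and locked twist, while handle has no constraint.
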PhysicsAsset /SparksCosmetics/Instrument/DrumKitHighVert/CountyFrog/Meshes/CountyFrog_DrumKit_Physics
bodies {
 dyn_sea_urchin_hihat [{shapeType=Sphere, radius=4.6061, shapeName=dyn_sea_urchin_hihat_sphere, center=<3.0671 0 0>}]
 handle [{shapeType=Capsule, length=0.505, radius=0.505, rotation=<rot 0 -90 0>}]
 hihat_top [{
   shapeType=Capsule
   length=76.0666
   radius=2.3508
   shapeName=hihat_top_capsule
   center=<-29.5373 0.046 0.7627>
   rotation=<rot 0 89.4405 -87.7651>
  }]
}
constraints [{bone1Name=dyn_sea_urchin_hihat, bone2Name=hihat_top, swing1LimitDegrees=12, swing2LimitDegrees=6, twistLimitDegrees=4, twistMotion=Locked}]
