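# Animation Blueprint for the CountyFrog drum kit cosmetic, animating idles, fidgets and blinks.
# A valid music clock with nonzero tempo selects beat-synced idle; otherwise a state machine plays idle or random fidgets, plus additive blinks.
# Delay loops drive fidgets and blinks, with cosmetic previews periodically substituting a reactivity demo using the beat-synced sequence.
# The result layers onto the input pose from dyn_spine, then passes through RigidBody physics in hi-hat cymbal bone space.
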AnimBlueprint /SparksCosmetics/Instrument/DrumKitHighVert/CountyFrog/Meshes/CountyFrog_DrumKit_AnimBP
target_skeleton=&/CountyFrog_DrumKit_Skeleton
graphs [
 AnimGraph {nodes [
   AnimGraphNode_StateMachine_0 = Main() {is=title AnimGraphNode_StateMachine}
   AnimGraphNode_StateMachine_1 = Blink() {is=title AnimGraphNode_StateMachine}
   n1 = Blend Poses by bool(
    BlendPose_0=Play Beat Synced Animation!(
     ExplicitTime:float=0
     bShouldLoop:bool=True
     InSequence:Anim Sequence Base Object Reference=Get BaseAnim_Beatsynced() -> (BaseAnim_Beatsynced:Anim Sequence Base Object Reference)
     MusicClock:Music Clock Object Reference
     TempoMultiplier:float=1
     bEnableTempoVariants:bool=False
     TempoVariantHysteresis:float=0.05
     bShowDebug:bool=False
    ) -> (Pose:Pose Link Structure) {is=title AnimGraphNode_PlayBeatSyncedAnim}
    BlendPose_1=Apply Additive(Base=@AnimGraphNode_StateMachine_0, Additive=@AnimGraphNode_StateMachine_1)
    bActiveValue=Get ClockValid() -> (ClockValid:bool)
   )
   n2 = Input Pose()
   n3 = RigidBody(ComponentPose=Local To Component(LocalPose=Layered blend per bone(BasePose=@n2, BlendPoses_0=@n1) {layerSetup [{branchFilters [{boneName=dyn_spine}]}]})) {
    bClampLinearTranslationLimitToRefPose=true
    bOverrideWorldGravity=true
    overridePhysicsAsset=&/CountyFrog_DrumKit_Physics
    simulationSpace=BaseBoneSpace
    alphaBoolBlend {blendOption=Linear}
    baseBoneRef {boneName=hihat_top_cymbal}
    componentLinearAccScale=<1 1 1>
    componentLinearVelScale=<0.2 0.2 0.2>
    overrideWorldGravity {z=-980}
    simSpaceSettings {dampingAlpha=0.5, worldAlpha=0.9}
   }
   Output Pose(Result=Component To Local(ComponentPose=@n3))
   n4 = Get BaseAnim_NoMusic() -> (BaseAnim_NoMusic:Anim Sequence Base Object Reference)
   n5 = Additive Identity Pose()
   Slot 'DefaultSlot'(Source=Apply Additive(Base=Sequence Player!(Sequence:Anim Sequence Base Object Reference=@n4) -> (Pose), Additive=Slot 'Additive'(Source=@n5) {slotName=Additive}))
  ]}
 Main {
  entry=Idle
  path=AnimGraph.AnimGraphNode_StateMachine_0.Main
  ids=id
  nodes [
   AnimStateNode_0 = Idle()
   AnimStateNode_1 = Breaker() {is=title AnimStateNode}
   AnimStateNode_2 = PreviewReactivity() {is=title AnimStateNode}
   AnimStateTransitionNode_0 = AnimStateTransitionNode() {bAllowInertializationForSelfTransitions=true}
   AnimStateTransitionNode_1 = AnimStateTransitionNode() {bAllowInertializationForSelfTransitions=true}
   AnimStateTransitionNode_2 = AnimStateTransitionNode() {bAllowInertializationForSelfTransitions=true}
   AnimStateTransitionNode_3 = AnimStateTransitionNode() {bAllowInertializationForSelfTransitions=true}
  ]
  links [
   {dst=Idle, rule=AnimGraph.AnimGraphNode_StateMachine_0.Main.AnimStateTransitionNode_1.Transition, src=Breaker, via=AnimStateTransitionNode_1}
   {dst=Breaker, rule=AnimGraph.AnimGraphNode_StateMachine_0.Main.AnimStateTransitionNode_0.Transition, src=Idle, via=AnimStateTransitionNode_0}
   {
    dst=PreviewReactivity
    rule=AnimGraph.AnimGraphNode_StateMachine_0.Main.AnimStateTransitionNode_2.Transition
    src=Idle
    via=AnimStateTransitionNode_2
   }
   {
    dst=Idle
    rule=AnimGraph.AnimGraphNode_StateMachine_0.Main.AnimStateTransitionNode_3.Transition
    src=PreviewReactivity
    via=AnimStateTransitionNode_3
   }
  ]
 }
 Idle {
  path=AnimGraph.AnimGraphNode_StateMachine_0.Main.AnimStateNode_0.Idle
  nodes [Output Animation Pose(Result=Sequence Player(Sequence=Get BaseAnim_NoMusic()))]
 }
 Breaker {
  path=AnimGraph.AnimGraphNode_StateMachine_0.Main.AnimStateNode_1.Breaker
  nodes [Output Animation Pose(Result=Sequence Player(Sequence=Get RandomBreaker() -> (RandomBreaker:Anim Sequence Object Reference)) {bLoopAnimation=false})]
 }
 PreviewReactivity {
  path=AnimGraph.AnimGraphNode_StateMachine_0.Main.AnimStateNode_2.PreviewReactivity
  nodes [Output Animation Pose(Result=Sequence Player 'CountyFrog_Drumkit_Idle_Beatsynced'() {sequence=&/CountyFrog_Drumkit_Idle_Beatsynced})]
 }
 Transition {
  path=AnimGraph.AnimGraphNode_StateMachine_0.Main.AnimStateTransitionNode_0.Transition
  nodes [Result(bCanEnterTransition=Get PlayBreaker() -> (PlayBreaker:bool))]
 }
 Transition {
  path=AnimGraph.AnimGraphNode_StateMachine_0.Main.AnimStateTransitionNode_1.Transition
  nodes [Result(bCanEnterTransition=NOT Boolean(A=Get PlayBreaker()))]
 }
 Transition {
  path=AnimGraph.AnimGraphNode_StateMachine_0.Main.AnimStateTransitionNode_2.Transition
  nodes [Result(bCanEnterTransition=Get Play Reactivity() -> (Play Reactivity:bool))]
 }
 Transition {
  path=AnimGraph.AnimGraphNode_StateMachine_0.Main.AnimStateTransitionNode_3.Transition
  nodes [Result(bCanEnterTransition=NOT Boolean(A=Get Play Reactivity()))]
 }
 Blink {
  entry=Idle
  path=AnimGraph.AnimGraphNode_StateMachine_1.Blink
  ids=id
  nodes [
   AnimStateNode_0 = Idle()
   AnimStateNode_1 = Blink() {is=title AnimStateNode}
   AnimStateTransitionNode_0 = AnimStateTransitionNode() {bAllowInertializationForSelfTransitions=true}
   AnimStateTransitionNode_1 = AnimStateTransitionNode() {bAllowInertializationForSelfTransitions=true}
  ]
  links [
   {dst=Idle, rule=AnimGraph.AnimGraphNode_StateMachine_1.Blink.AnimStateTransitionNode_1.Transition, src=Blink, via=AnimStateTransitionNode_1}
   {dst=Blink, rule=AnimGraph.AnimGraphNode_StateMachine_1.Blink.AnimStateTransitionNode_0.Transition, src=Idle, via=AnimStateTransitionNode_0}
  ]
 }
 Idle {path=AnimGraph.AnimGraphNode_StateMachine_1.Blink.AnimStateNode_0.Idle, nodes [Output Animation Pose(Result=Additive Identity Pose())]}
 Blink {
  path=AnimGraph.AnimGraphNode_StateMachine_1.Blink.AnimStateNode_1.Blink
  nodes [Output Animation Pose(Result="Sequence Player (additive) 'CountyFrog_Drumkit_Blink'"() {bLoopAnimation=false, sequence=&/CountyFrog_Drumkit_Blink}) {
    node {stateExitFunction {functionName=OnStateExit}}
    stateExitFunction {bSelfContext=true, memberGuid=#1, memberName=OnStateExit}
   }]
 }
 Transition {
  path=AnimGraph.AnimGraphNode_StateMachine_1.Blink.AnimStateTransitionNode_0.Transition
  nodes [Result(bCanEnterTransition=Get PlayBlink() -> (PlayBlink:bool))]
 }
 Transition {
  path=AnimGraph.AnimGraphNode_StateMachine_1.Blink.AnimStateTransitionNode_1.Transition
  nodes [Result(bCanEnterTransition="float <= float"(A="Get Relevant Anim Time Remaining (Blink)"!() -> (ReturnValue:float) {is=title K2Node_AnimGetter}) -> (ReturnValue=~))]
 }
 ChangeRandomBreaker {nodes [
   ChangeRandomBreaker() {is=title K2Node_FunctionEntry}
   > n1 = Set BreakerListLocal(BreakerListLocal:Array of Anim Sequence Object References=Get BreakerList() -> (BreakerList:Array of Anim Sequence Object References)) -> (Output_Get:Array of Anim Sequence Object References)
   > Remove Item!(execute, TargetArray:Array of Anim Sequence Object References=@n1, Item:"Anim Sequence Object Reference (by ref)"=Get RandomBreaker()) -> (then, ReturnValue:bool)
   > n2 = Set RandomBreaker(RandomBreaker:Anim Sequence Object Reference=Random Array Item!(TargetArray:Array of Anim Sequence Object References=@n1) -> (OutItem:Anim Sequence Object Reference, OutIndex:int=0) {is=title K2Node_CallArrayFunction}) -> (Output_Get:Anim Sequence Object Reference)
   > Set RandomBreakerLength(RandomBreakerLength:float=Get SequenceLength!(self:Anim Sequence Base Object Reference=@n2) -> (SequenceLength:float)) -> (Output_Get:float)
  ]}
 EventGraph {nodes [
   Event BlueprintUpdateAnimation()
   n1 = GetHighestPriorityClock!(self:Music Event Subsystem Object Reference=Get MusicEventSubsystem() -> (ReturnValue:Music Event Subsystem Object Reference) {is=title K2Node_GetSubsystem}) -> (ReturnValue:Music Clock Object Reference)
   > n2 = Is Valid(InputObject=@n1)
   n3 = Set Music Clock(execute=@n2.Is Not Valid, Music Clock:Music Clock Object Reference) -> (Output_Get:Music Clock Object Reference)
   n4 = GetCurrentTempo!(self:Music Clock Object Reference=@n1) -> (ReturnValue:float)
   Set ClockValid(execute=[@n5.then, @n3.then], ClockValid:bool=AND Boolean(A=IsValid(Object=@n1), B="Not Equal (Float)"(A=@n4))) -> (Output_Get:bool)
   n5 = Set Music Clock(execute=@n2.Is Valid, Music Clock=@n1)
   Event BlueprintInitializeAnimation()
   > n6 = Sequence()
   Delay(execute=[@n6.then_0, @n9.then, @n10.then], Duration=RandomFloatInRange(Min=2, Max=5))
   > n7 = Branch(Condition=IsCosmeticObjectDisplayedAsPreview!(CosmeticObject:Object Reference=GetOwningActor()) -> (ReturnValue:bool))
   > n8 = MultiGate!(execute:Exec, Reset:Exec, IsRandom:bool, Loop:bool=true, StartIndex:int=-1) -> (Out 0:Exec, Out 1:Exec, Out 2:Exec) {is=title K2Node_MultiGate}
   ChangeRandomBreaker(execute=[@n7.else, @n8.Out 0, @n8.Out 1]) {is=title K2Node_CallFunction}
   > Set PlayBreaker(PlayBreaker:bool=true) -> (Output_Get:bool)
   > Delay(Duration=Get RandomBreakerLength() -> (RandomBreakerLength:float))
   > n9 = Set PlayBreaker(PlayBreaker=false)
   Set Play Reactivity(execute=@n8.Out 2, Play Reactivity:bool=true) -> (Output_Get:bool)
   > Delay(Duration=3.86)
   > n10 = Set Play Reactivity(Play Reactivity=false)
   Delay(execute=[@n6.then_1, @n13.then], Duration=RandomFloatInRange(Min=2, Max=4.2))
   n11 = Get PlayBreaker()
   n12 = Get Play Reactivity()
   > n13 = Set PlayBlink(PlayBlink:bool=NOT Boolean(A=OR Boolean(A=@n11, B=@n12))) -> (Output_Get:bool)
  ]}
 OnStateExitBlink {nodes [
   OnStateExitBlink!() -> (then, Context:"Anim Update Context Structure (by ref)", Node:"Anim Node Reference Structure (by ref)") {is=title K2Node_FunctionEntry}
   > Set PlayBlink(PlayBlink=false)
   > Return Node()
  ]}
]
variables [
 BaseAnim_Beatsynced {type=AnimSequenceBase, default=&CountyFrog_Drumkit_Idle_Beatsynced}
 BaseAnim_NoMusic {type=AnimSequenceBase, default=&/CountyFrog_Drumkit_Idle.CountyFrog_DrumKit_Idle}
 BreakerIndex {type=integer}
 BreakerList {
  type="array<AnimSequence>"
  default [
   &/CountyFrog_Drumkit_Fidget_A
   &/CountyFrog_Drumkit_Fidget_B
   &/CountyFrog_DrumKit_Fidget_C
  ]
 }
 ClockValid {type=boolean}
 IdleBreakers {type="array<AnimMontage>", default []}
 Music Clock {type=MusicClockComponent, default=null}
 Play Reactivity {type=boolean}
 PlayBlink {type=boolean}
 PlayBreaker {type=boolean}
 RandomBreaker {type=AnimSequence, default=&CountyFrog_Drumkit_Fidget_A}
 RandomBreakerLength {type=number}
 StopBreaker {type=boolean}
]
